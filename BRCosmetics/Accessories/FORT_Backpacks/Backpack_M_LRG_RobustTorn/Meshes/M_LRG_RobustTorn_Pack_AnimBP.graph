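# Animation Blueprint for the RobustTorn backpack cosmetic that shakes its two rattle bones as a reaction to enemy eliminations and chest openings.
# Each update eases an intensity toward the reaction flag, scaling randomized, sine-of-time jitter into asymmetric rotators.
# AnimGraph applies these additively in bone space to the left and right bones.
# Front-end display replays the reaction on a looping timer, and customization load sets ReactivityEnabled from the Stage1 variant.
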AnimBlueprint /BRCosmetics/Accessories/FORT_Backpacks/Backpack_M_LRG_RobustTorn/Meshes/M_LRG_RobustTorn_Pack_AnimBP
target_skeleton=&/M_LRG_RobustTorn_Pack_Skeleton
graphs [
 AnimGraph {nodes [
   n1 = "Transform (Modify) Bone - Bone: dyn_rattle_base_l"(
    ComponentPose=Local To Component(LocalPose=Input Pose())
    Alpha=1
    Translation="0,0,0"
    Rotation=Get RattleJittering_L() -> (RattleJittering_L:Rotator)
    Scale="1,1,1"
   ) {
    rotationMode=BMM_Additive
    rotationSpace=BCS_BoneSpace
    scaleMode=BMM_Additive
    translationMode=BMM_Additive
    translationSpace=BCS_BoneSpace
    alphaBoolBlend {blendOption=Linear}
    boneToModify {boneName=dyn_rattle_base_l}
   }
   n2 = "Transform (Modify) Bone - Bone: dyn_rattle_base_r"(ComponentPose=@n1, Alpha=1, Translation="0,0,0", Rotation=Get RattleJittering_R() -> (RattleJittering_R:Rotator), Scale="1,1,1") {
    rotationMode=BMM_Additive
    rotationSpace=BCS_BoneSpace
    scaleMode=BMM_Additive
    translationMode=BMM_Additive
    translationSpace=BCS_BoneSpace
    alphaBoolBlend {blendOption=Linear}
    boneToModify {boneName=dyn_rattle_base_r}
   }
   Output Pose(Result=Component To Local(ComponentPose=@n2)) {layerGroup=None}
  ]}
 EventGraph {nodes [
   n1 = Reactive() {is=title K2Node_CustomEvent}
   > Branch(Condition=Get bIsInFrontEnd())
   > Delay(Duration=2)
   > Set bEnemyPlayerEliminated(bEnemyPlayerEliminated:bool=true) -> (Output_Get:bool)
   > Delay(Duration=4)
   > Set bEnemyPlayerEliminated(bEnemyPlayerEliminated=false)
   n2 = Event BlueprintUpdateAnimation()
   > Set Delta Time X(Delta Time X=@n2.DeltaTimeX)
   n3 = Get bEnemyPlayerEliminated() -> (bEnemyPlayerEliminated:bool)
   > n4 = Set Acceleration(Acceleration:double=FInterpTo(Current=Get Acceleration() -> (Acceleration:double), Target=SelectFloat(A=2, B=0, bPickA=@n3), DeltaTime=Get Delta Time X(), InterpSpeed=2)) -> (Output_Get:double)
   n5 = float * float(A=float * float(A=10, B=100), B=GetGameTimeInSeconds())
   n6 = RandomFloatInRange(Min=float * float(A="Sin (Radians)"(A=@n5), B=@n4), Max=0)
   n7 = MakeRotator!(Roll, Pitch=@n6, Yaw=@n6) -> (ReturnValue_Roll:float, ReturnValue_Pitch:float, ReturnValue_Yaw:float)
   > Set!(
    execute
    RattleJittering_L_Roll:float=0
    RattleJittering_L_Pitch:float=float * float(A=@n7.ReturnValue_Pitch, B=0.5) {is=title K2Node_PromotableOperator}
    RattleJittering_L_Yaw:float=@n7.ReturnValue_Yaw
   ) -> (then, Output_Get:Rotator)
   > Set!(
    execute
    RattleJittering_R_Roll:float=0
    RattleJittering_R_Pitch:float=float * float(A=@n7.ReturnValue_Pitch, B=0.5)
    RattleJittering_R_Yaw:float=float * float(A=@n7.ReturnValue_Yaw, B=1.3)
   ) -> (then, Output_Get)
   Event OnCharacterCustomizationLoaded()
   > n8 = Cast To FortPlayerPawn(Object=GetOwningActor())
   > n9 = GetVariantChannelActiveIndexForItemActor!(
    execute
    Item:Athena Cosmetic Item Definition Object Reference=/BRCosmetics/Athena/Items/Cosmetics/Backpacks/Backpack_RobustTorn
    Actor:Actor Object Reference=@n8
    Channel:Gameplay Tag Structure=(TagName="Cosmetics.Variant.Channel.Parts")
    VariantList:Array of Gameplay Tag Structures=Make Array("[0]"=(TagName="Cosmetics.Variant.Property.Stage1"))
    bErrorIfChannelMissing:bool=false
    bErrorIfTagNotInArray:bool=false
   ) -> (then, Index:int=0, bIndexIsValid:bool, bFoundChannel:bool)
   > n10 = Branch(Condition=@n9.bIndexIsValid)
   > Set ReactivityEnabled(ReactivityEnabled:bool=true) -> (Output_Get:bool)
   > Reactive() {is=title K2Node_CallFunction}
   Set ReactivityEnabled(execute=@n10.else, ReactivityEnabled=false)
   n11 = Event BlueprintProcessAnimInputEvent()
   > Branch(Condition=OR Boolean(A=IsNamed(self=@n11.AnimInputEvent, InEventName=aie.EnemyPlayerEliminated), B=IsNamed(self=@n11.AnimInputEvent, InEventName=aie.ChestOpened)))
   > Set bEnemyPlayerEliminated(bEnemyPlayerEliminated=true)
   > Reactive()
   > Delay(Duration=4)
   > Set bEnemyPlayerEliminated(bEnemyPlayerEliminated=false)
   Event BlueprintBeginPlay() {is=title K2Node_Event}
   > Parent: BlueprintBeginPlay() {is=title K2Node_CallParentFunction}
   > Branch(Condition=IsCosmeticObjectDisplayedInFrontendContext!(CosmeticObject:Object Reference) -> (ReturnValue:bool))
   > Delay(Duration=4)
   > Reactive()
   > Set Timer by Event(Delegate=@n1, Time=13, bLooping=true)
   TryGetPawnOwner()
  ]}
]
variables [
 Acceleration {type=number}
 bEnemyPlayerEliminated {type=boolean}
 Delta Time X {type=number}
 RattleJittering_L {type=Rotator}
 RattleJittering_R {type=Rotator}
 ReactivityEnabled {type=boolean}
]
